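# UI material for the Thrive item shop background, blending a color gradient with a background texture.
# UVs are scaled about their center and rotated by scale and angle parameters, and the red component drives a curve atlas row lookup into the Thrive background curve.
# The resulting gradient feeds a named reroute, and a fixed-alpha lerp mixes a reroute value with the Thrive texture, weighted heavily toward the texture.
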Material /ItemShop/UI/Art/Thrive/M_UI_ThriveShopBackground
graph {
 properties {bUsedWithStaticMesh=false, materialDomain=MD_UI, pixelDepthOffsetMode=PDOM_Legacy, usedShadingModels=Default Lit}
 nodes [
  LinearInterpolate(A=NamedRerouteUsage(), B=TextureSample() {texture=&/T_UI_Thrive_BG}.RGB) {constAlpha=0.8907}
  n1 = MaterialFunctionCall(Texture Scale=ScalarParameter() {defaultValue=2.0753, group=Gradient, parameterName=GradientScale}) {materialFunction=&ScaleUVsByCenter}
  n2 = ScalarParameter() {defaultValue=1.2149, group=Gradient, parameterName=GradientAngle}
  n3 = CurveAtlasRowParameter(CurveTime=ComponentMask(None=MaterialFunctionCall(UVs=@n1.UVs, "Rotation Angle (0-1)"=@n2) {materialFunction=&CustomRotator}.Rotated Values) {r=true}) {
   atlas=/Game/UI/Frontend/ItemShop/Art/ItemShopColorAtlas
   curve=/Game/UI/Frontend/ItemShop/Art/Thrive/ItemShop_ThriveBG_Curve
   defaultValue=7
   group=Gradient
   parameterName=CurveParamsBG
  }
  NamedRerouteDeclaration(None=@n3) {name=Gradient_PY, nodeColor=<rgba 0.9911 0.1714 1 1>}
 ]
}
